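# Physics asset for the ProdigyHaughty male medium body, covering the lower body, coat tails and left-thigh spikes.
# Capsules on pelvis, thighs and calves form the body, with free-rotating joints between them.
# Six coat chains each run from a tiny base capsule on the pelvis through three boxes, bending mainly on swing2 with zero twist limit.
# Three spike capsules chain off the left thigh, and the last also constrains back to it, closing a loop.
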PhysicsAsset /BRCosmetics/Characters/Player/Male/Medium/Bodies/M_MED_ProdigyHaughty/Meshes/M_MED_ProdigyHaughty_Physics
bodies {
 calf_l [{shapeType=Capsule, length=35.5, radius=8, shapeName=calf_l_capsule, center=<-14.7999 -2.0005 0>, rotation=<rot -90 -96.7215 -78.2793>}]
 calf_r [{shapeType=Capsule, length=41, radius=9, shapeName=calf_r_capsule, center=<16.2081 0.1064 -1.1931>, rotation=<rot -90 -179.6419 -175.3587>}]
 dyn_coat_bck_l_01 [{shapeType=Box, extentX=13.5, extentY=13.5, extentZ=3.5, shapeName=dyn_coat_bck_l_01_box, center=<6.2114 0 0.4959>}]
 dyn_coat_bck_l_02 [{shapeType=Box, extentX=13.5, extentY=13.5, extentZ=3.5, shapeName=dyn_coat_bck_l_02_box, center=<6.2114 0 0.4959>}]
 dyn_coat_bck_l_03 [{shapeType=Box, extentX=19.5, extentY=21.5, extentZ=10.5, shapeName=dyn_coat_bck_l_03_box, center=<8.495 0 -1.4889>}]
 dyn_coat_bck_l_base [{shapeType=Capsule, length=0.505, radius=0.505, shapeName=dyn_coat_bck_l_base_capsule}]
 dyn_coat_bck_r_01 [{shapeType=Box, extentX=13.5, extentY=13.5, extentZ=3.5, shapeName=dyn_coat_bck_r_01_box, center=<6.2114 0 0.4959>}]
 dyn_coat_bck_r_02 [{shapeType=Box, extentX=13.5, extentY=13.5, extentZ=3.5, shapeName=dyn_coat_bck_r_02_box, center=<6.2114 0 0.4959>}]
 dyn_coat_bck_r_03 [{shapeType=Box, extentX=18.5, extentY=31.5, extentZ=6.5, shapeName=dyn_coat_bck_r_03_box, center=<9.8488 -1.9686 -1.3173>}]
 dyn_coat_bck_r_base [{shapeType=Capsule, length=0.505, radius=0.505, shapeName=dyn_coat_bck_r_base_capsule}]
 dyn_coat_fr_l_01 [{shapeType=Box, extentX=21.5, extentY=18.5, extentZ=5.5, shapeName=dyn_coat_fr_l_01_box, center=<11.3506 0 2.2944>}]
 dyn_coat_fr_l_02 [{shapeType=Box, extentX=15.5, extentY=14.5, extentZ=6.5, shapeName=dyn_coat_fr_l_02_box, center=<10.808 0 -2.1782>}]
 dyn_coat_fr_l_03 [{shapeType=Box, extentX=17.5, extentY=19.5, extentZ=7.5, shapeName=dyn_coat_fr_l_03_box, center=<9.4628 -1.7708 -1.3647>}]
 dyn_coat_fr_l_base [{shapeType=Capsule, length=0.505, radius=0.505, shapeName=dyn_coat_fr_l_base_capsule}]
 dyn_coat_fr_r_01 [{shapeType=Box, extentX=11.5, extentY=11.5, extentZ=3.5, shapeName=dyn_coat_fr_r_01_box, center=<9.4628 0 0.3685>}]
 dyn_coat_fr_r_02 [{
   shapeType=Box
   extentX=13.5
   extentY=15.5
   extentZ=6.5
   shapeName=dyn_coat_fr_r_02_box
   center=<9.4628 0 0.0447>
   rotation=<rot -10 -0.0001 0>
  }]
 dyn_coat_fr_r_03 [{
   shapeType=Box
   extentX=17.5
   extentY=18.5
   extentZ=5.5
   shapeName=dyn_coat_fr_r_03_box
   center=<9.4628 0.6399 -0.4842>
   rotation=<rot 0.0001 -0.0002 20>
  }]
 dyn_coat_fr_r_base [{shapeType=Capsule, length=0.505, radius=0.505, shapeName=dyn_coat_fr_r_base_capsule}]
 dyn_coat_side_l_01 [{shapeType=Box, extentX=13.5, extentY=13.5, extentZ=3.5, shapeName=dyn_coat_side_l_01_box, center=<6.2114 0 0.4959>}]
 dyn_coat_side_l_02 [{shapeType=Box, extentX=13.5, extentY=13.5, extentZ=3.5, shapeName=dyn_coat_side_l_02_box, center=<6.2114 0 0.4959>}]
 dyn_coat_side_l_03 [{shapeType=Box, extentX=19.5, extentY=17.5, extentZ=3.5, shapeName=dyn_coat_side_l_03_box, center=<7.2365 0 0.4959>}]
 dyn_coat_side_l_base [{shapeType=Capsule, length=0.505, radius=0.505, shapeName=dyn_coat_side_l_base_capsule}]
 dyn_coat_side_r_01 [{shapeType=Box, extentX=13.5, extentY=13.5, extentZ=3.5, shapeName=dyn_coat_side_r_01_box, center=<6.2114 0 0.4959>}]
 dyn_coat_side_r_02 [{shapeType=Box, extentX=13.5, extentY=13.5, extentZ=5.5, shapeName=dyn_coat_side_r_02_box, center=<6.2114 0 0.4959>}]
 dyn_coat_side_r_03 [{
   shapeType=Box
   extentX=20.5
   extentY=35.5
   extentZ=7.5
   shapeName=dyn_coat_side_r_03_box
   center=<9.4654 -1.8402 -1.9095>
   rotation=<rot -0.0002 -0.0003 -9.9999>
  }]
 dyn_coat_side_r_base [{shapeType=Capsule, length=0.505, radius=0.505, shapeName=dyn_coat_side_r_base_capsule}]
 dyn_spike_bck [{shapeType=Capsule, length=8.5, radius=2, shapeName=dyn_spike_bck_capsule, center=<0 -2.968 0.294>, rotation=<rot 0.0001 -0.0003 -79.9999>}]
 dyn_spike_fr [{shapeType=Capsule, length=7.5, radius=2, shapeName=dyn_spike_fr_capsule, center=<-1.4612 -0.6649 -2.6442>, rotation=<rot -0.0003 -0.0001 12>}]
 dyn_spike_mid [{
   shapeType=Capsule
   length=9
   radius=1.5
   shapeName=dyn_spike_mid_capsule
   center=<-1.2828 -2.926 -1.6085>
   rotation=<rot -0.2936 -10.8201 50.9261>
  }]
 pelvis [{shapeType=Capsule, length=6.5, radius=11, shapeName=pelvis_capsule, center=<-4.7074 -0.8249 0>}]
 thigh_l [{shapeType=Capsule, length=24, radius=8.5, shapeName=thigh_l_capsule, center=<-16.0977 0.7625 0.4591>, rotation=<rot 90 0 0.0012>}]
 thigh_r [{
   shapeType=Capsule
   length=24.5
   radius=9.5
   shapeName=thigh_r_capsule
   center=<14.8974 0.4686 -0.6349>
   rotation=<rot -90 -179.9983 179.9979>
  }]
}
constraints [
 {
  bone1Name=thigh_l
  bone2Name=pelvis
  swing1LimitDegrees=45
  swing1Motion=Free
  swing2LimitDegrees=45
  swing2Motion=Free
  twistLimitDegrees=45
  twistMotion=Free
 }
 {
  bone1Name=calf_l
  bone2Name=thigh_l
  swing1LimitDegrees=45
  swing1Motion=Free
  swing2LimitDegrees=45
  swing2Motion=Free
  twistLimitDegrees=45
  twistMotion=Free
 }
 {
  bone1Name=thigh_r
  bone2Name=pelvis
  swing1LimitDegrees=45
  swing1Motion=Free
  swing2LimitDegrees=45
  swing2Motion=Free
  twistLimitDegrees=45
  twistMotion=Free
 }
 {
  bone1Name=calf_r
  bone2Name=thigh_r
  swing1LimitDegrees=45
  swing1Motion=Free
  swing2LimitDegrees=45
  swing2Motion=Free
  twistLimitDegrees=45
  twistMotion=Free
 }
 {bone1Name=dyn_coat_bck_l_02, bone2Name=dyn_coat_bck_l_01, swing1LimitDegrees=0, swing2LimitDegrees=50, twistLimitDegrees=0}
 {bone1Name=dyn_coat_bck_l_03, bone2Name=dyn_coat_bck_l_02, swing1LimitDegrees=0, swing2LimitDegrees=50, twistLimitDegrees=0}
 {bone1Name=dyn_coat_bck_r_02, bone2Name=dyn_coat_bck_r_01, swing1LimitDegrees=0, swing2LimitDegrees=50, twistLimitDegrees=0}
 {bone1Name=dyn_coat_bck_r_03, bone2Name=dyn_coat_bck_r_02, swing1LimitDegrees=0, swing2LimitDegrees=50, twistLimitDegrees=0}
 {bone1Name=dyn_coat_fr_l_02, bone2Name=dyn_coat_fr_l_01, swing1LimitDegrees=5, swing2LimitDegrees=50, twistLimitDegrees=0}
 {bone1Name=dyn_coat_fr_l_03, bone2Name=dyn_coat_fr_l_02, swing1LimitDegrees=5, swing2LimitDegrees=50, twistLimitDegrees=0}
 {bone1Name=dyn_coat_fr_r_02, bone2Name=dyn_coat_fr_r_01, swing1LimitDegrees=5, swing2LimitDegrees=50, twistLimitDegrees=0}
 {bone1Name=dyn_coat_fr_r_03, bone2Name=dyn_coat_fr_r_02, swing1LimitDegrees=5, swing2LimitDegrees=50, twistLimitDegrees=0}
 {bone1Name=dyn_coat_side_l_02, bone2Name=dyn_coat_side_l_01, swing1LimitDegrees=5, swing2LimitDegrees=50, twistLimitDegrees=0}
 {bone1Name=dyn_coat_side_l_03, bone2Name=dyn_coat_side_l_02, swing1LimitDegrees=5, swing2LimitDegrees=50, twistLimitDegrees=0}
 {bone1Name=dyn_coat_side_r_02, bone2Name=dyn_coat_side_r_01, swing1LimitDegrees=5, swing2LimitDegrees=50, twistLimitDegrees=0}
 {bone1Name=dyn_coat_side_r_03, bone2Name=dyn_coat_side_r_02, swing1LimitDegrees=5, swing2LimitDegrees=50, twistLimitDegrees=0}
 {bone1Name=dyn_spike_fr, bone2Name=thigh_l, swing1LimitDegrees=15, swing2LimitDegrees=15, twistLimitDegrees=0}
 {bone1Name=dyn_spike_mid, bone2Name=dyn_spike_fr, swing1LimitDegrees=15, swing2LimitDegrees=15, twistLimitDegrees=0}
 {bone1Name=dyn_spike_bck, bone2Name=dyn_spike_mid, swing1LimitDegrees=15, swing2LimitDegrees=15, twistLimitDegrees=0}
 {bone1Name=dyn_spike_bck, bone2Name=thigh_l, swing1LimitDegrees=0, swing2LimitDegrees=0, twistLimitDegrees=0}
 {bone1Name=dyn_coat_bck_l_01, bone2Name=dyn_coat_bck_l_base, swing1LimitDegrees=0, swing2LimitDegrees=35, twistLimitDegrees=0}
 {
  bone1Name=dyn_coat_bck_l_base
  bone2Name=pelvis
  swing1LimitDegrees=45
  swing1Motion=Free
  swing2LimitDegrees=45
  swing2Motion=Free
  twistLimitDegrees=45
  twistMotion=Free
 }
 {bone1Name=dyn_coat_bck_r_01, bone2Name=dyn_coat_bck_r_base, swing1LimitDegrees=0, swing2LimitDegrees=35, twistLimitDegrees=0}
 {
  bone1Name=dyn_coat_bck_r_base
  bone2Name=pelvis
  swing1LimitDegrees=45
  swing1Motion=Free
  swing2LimitDegrees=45
  swing2Motion=Free
  twistLimitDegrees=45
  twistMotion=Free
 }
 {bone1Name=dyn_coat_fr_l_01, bone2Name=dyn_coat_fr_l_base, swing1LimitDegrees=0, swing2LimitDegrees=35, twistLimitDegrees=0}
 {
  bone1Name=dyn_coat_fr_l_base
  bone2Name=pelvis
  swing1LimitDegrees=45
  swing1Motion=Free
  swing2LimitDegrees=45
  swing2Motion=Free
  twistLimitDegrees=45
  twistMotion=Free
 }
 {bone1Name=dyn_coat_fr_r_01, bone2Name=dyn_coat_fr_r_base, swing1LimitDegrees=0, swing2LimitDegrees=35, twistLimitDegrees=0}
 {
  bone1Name=dyn_coat_fr_r_base
  bone2Name=pelvis
  swing1LimitDegrees=45
  swing1Motion=Free
  swing2LimitDegrees=45
  swing2Motion=Free
  twistLimitDegrees=45
  twistMotion=Free
 }
 {bone1Name=dyn_coat_side_l_01, bone2Name=dyn_coat_side_l_base, swing1LimitDegrees=0, swing2LimitDegrees=35, twistLimitDegrees=0}
 {
  bone1Name=dyn_coat_side_l_base
  bone2Name=pelvis
  swing1LimitDegrees=45
  swing1Motion=Free
  swing2LimitDegrees=45
  swing2Motion=Free
  twistLimitDegrees=45
  twistMotion=Free
 }
 {bone1Name=dyn_coat_side_r_01, bone2Name=dyn_coat_side_r_base, swing1LimitDegrees=0, swing2LimitDegrees=35, twistLimitDegrees=0}
 {
  bone1Name=dyn_coat_side_r_base
  bone2Name=pelvis
  swing1LimitDegrees=45
  swing1Motion=Free
  swing2LimitDegrees=45
  swing2Motion=Free
  twistLimitDegrees=45
  twistMotion=Free
 }
]
