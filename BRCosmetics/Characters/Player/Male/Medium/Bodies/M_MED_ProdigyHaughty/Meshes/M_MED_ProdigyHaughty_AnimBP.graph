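# Post-process style AnimBP for the ProdigyHaughty medium male body cosmetic, adding secondary motion to coat, lapels, necklace and bracelets.
# The parent's copied pose runs through a deform Control Rig, RigidBody, constraints and jewelry AnimDynamics into a cached pose.
# That cache feeds additive coat and spike bone offsets, then per-bone coat AnimDynamics.
# The EventGraph computes emote-dependent gravity, crouch and thigh-angle values, with no visible AnimGraph reads.
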AnimBlueprint /BRCosmetics/Characters/Player/Male/Medium/Bodies/M_MED_ProdigyHaughty/Meshes/M_MED_ProdigyHaughty_AnimBP
target_skeleton=&/M_MED_ProdigyHaughty_Skeleton
graphs [
 AnimGraph {nodes [
   n1 = "Transform (Modify) Bone - Bone: dyn_coat_fr_r_base"(Alpha=1, Translation="-2.048,8.260,0.227", Rotation="15.012,-58.305,13.216", Scale="1,1,1") {
    rotationMode=BMM_Additive
    rotationSpace=BCS_BoneSpace
    translationMode=BMM_Additive
    translationSpace=BCS_BoneSpace
    alphaBoolBlend {blendOption=Linear}
    alphaScaleBiasClamp {bInterpResult=true}
    boneToModify {boneName=dyn_coat_fr_r_base}
    rotation=<rot 15.012 -58.3051 13.216>
    translation=<-2.048 8.26 0.227>
   }
   n2 = "Transform (Modify) Bone - Bone: dyn_coat_side_r_base"(ComponentPose=@n1, Alpha=1, Translation="0,0,0", Rotation="0,0,0", Scale="1,1,1") {
    rotationMode=BMM_Additive
    rotationSpace=BCS_BoneSpace
    translationMode=BMM_Additive
    translationSpace=BCS_BoneSpace
    alphaBoolBlend {blendOption=Linear}
    alphaScaleBiasClamp {bInterpResult=true}
    boneToModify {boneName=dyn_coat_side_r_base}
   }
   n3 = "Transform (Modify) Bone - Bone: dyn_coat_fr_r_01"(ComponentPose=@n2, Alpha=1, Translation="0,0,0", Rotation="-10.000,-0.000,0.000", Scale="1,1,1") {
    rotationMode=BMM_Additive
    rotationSpace=BCS_BoneSpace
    translationMode=BMM_Additive
    translationSpace=BCS_BoneSpace
    alphaBoolBlend {blendOption=Linear}
    alphaScaleBiasClamp {bInterpResult=true}
    boneToModify {boneName=dyn_coat_fr_r_01}
    rotation=<rot -10 0 0.0001>
   }
   n4 = "Transform (Modify) Bone - Bone: dyn_coat_fr_r_02"(ComponentPose=@n3, Alpha=1, Translation="0,0,0", Rotation="-20.000,-0.010,-0.031", Scale="1,1,1") {
    rotationMode=BMM_Additive
    rotationSpace=BCS_BoneSpace
    translationMode=BMM_Additive
    translationSpace=BCS_BoneSpace
    alphaBoolBlend {blendOption=Linear}
    alphaScaleBiasClamp {bInterpResult=true}
    boneToModify {boneName=dyn_coat_fr_r_02}
    rotation=<rot -20 -0.01 -0.031>
   }
   n5 = "Transform (Modify) Bone - Bone: dyn_coat_fr_r_base"(ComponentPose=@n4, Alpha=1, Translation="0,0,0", Rotation="0,0,0", Scale="1,1,1") {
    rotationMode=BMM_Additive
    rotationSpace=BCS_BoneSpace
    translationMode=BMM_Additive
    translationSpace=BCS_BoneSpace
    alphaBoolBlend {blendOption=Linear}
    alphaScaleBiasClamp {bInterpResult=true}
    boneToModify {boneName=dyn_coat_fr_r_base}
   }
   "Transform (Modify) Bone - Bone: dyn_coat_fr_r_02"(ComponentPose=@n5, Alpha=1, Translation="0,0,0", Rotation="0,0,0", Scale="1,1,1") {
    rotationMode=BMM_Additive
    rotationSpace=BCS_BoneSpace
    translationMode=BMM_Additive
    translationSpace=BCS_BoneSpace
    alphaBoolBlend {blendOption=Linear}
    alphaScaleBiasClamp {bInterpResult=true}
    boneToModify {boneName=dyn_coat_fr_r_02}
   }
   "Transform (Modify) Bone - Bone: dyn_spike_fr"(Alpha=1, Translation="-2.236,-3.571,-0.022", Rotation="0,0,0", Scale="1,1,1") {
    rotationMode=BMM_Additive
    rotationSpace=BCS_BoneSpace
    translationMode=BMM_Additive
    translationSpace=BCS_BoneSpace
    alphaBoolBlend {blendOption=Linear}
    alphaScaleBiasClamp {bInterpResult=true, interpSpeedDecreasing=25, interpSpeedIncreasing=25}
    boneToModify {boneName=dyn_spike_fr}
    translation=<-2.236 -3.571 -0.022>
   }
   "Transform (Modify) Bone - Bone: dyn_spike_fr"(Alpha=1, Translation="0,0,0", Rotation="0,0,0", Scale="1,1,1") {
    rotationMode=BMM_Additive
    rotationSpace=BCS_BoneSpace
    translationMode=BMM_Additive
    translationSpace=BCS_BoneSpace
    alphaBoolBlend {blendOption=Linear}
    alphaScaleBiasClamp {bInterpResult=true}
    boneToModify {boneName=dyn_spike_fr}
   }
   n6 = "Transform (Modify) Bone - Bone: dyn_coat_fr_l_base"(Alpha=1, Translation="0,0,0.5", Rotation="0,0,0", Scale="1,1,1") {
    rotationMode=BMM_Additive
    rotationSpace=BCS_BoneSpace
    translationMode=BMM_Additive
    translationSpace=BCS_BoneSpace
    alphaBoolBlend {blendOption=Linear}
    alphaScaleBiasClamp {bInterpResult=true}
    boneToModify {boneName=dyn_coat_fr_l_base}
    translation {z=0.5}
   }
   n7 = "Transform (Modify) Bone - Bone: dyn_coat_fr_l_base"(ComponentPose=@n6, Alpha=1, Translation="-5.044,-0.109,1.908", Rotation="-20.785,-5.984,-10.820", Scale="1,1,1") {
    rotationMode=BMM_Additive
    rotationSpace=BCS_BoneSpace
    translationMode=BMM_Additive
    translationSpace=BCS_BoneSpace
    alphaBoolBlend {blendOption=Linear}
    alphaScaleBiasClamp {bInterpResult=true, interpSpeedDecreasing=25, interpSpeedIncreasing=25}
    boneToModify {boneName=dyn_coat_fr_l_base}
    rotation=<rot -20.785 -5.984 -10.82>
    translation=<-5.044 -0.109 1.908>
   }
   n8 = "Transform (Modify) Bone - Bone: dyn_coat_side_l_base"(ComponentPose=@n7, Alpha=1, Translation="-1.152,-0.590,0.604", Rotation="0,0,0", Scale="1,1,1") {
    rotationMode=BMM_Additive
    rotationSpace=BCS_BoneSpace
    translationMode=BMM_Additive
    translationSpace=BCS_BoneSpace
    alphaBoolBlend {blendOption=Linear}
    alphaScaleBiasClamp {bInterpResult=true, interpSpeedDecreasing=25, interpSpeedIncreasing=25}
    boneToModify {boneName=dyn_coat_side_l_base}
    translation=<-1.152 -0.59 0.604>
   }
   n9 = "Transform (Modify) Bone - Bone: dyn_coat_fr_l_base"(ComponentPose=@n8, Alpha=1, Translation="1.933,0.140,0.004", Rotation="19.873,-10.421,-1.849", Scale="1,1,1") {
    rotationMode=BMM_Additive
    rotationSpace=BCS_BoneSpace
    translationMode=BMM_Additive
    translationSpace=BCS_BoneSpace
    alphaBoolBlend {blendOption=Linear}
    alphaScaleBiasClamp {bInterpResult=true, interpSpeedDecreasing=25, interpSpeedIncreasing=25}
    boneToModify {boneName=dyn_coat_fr_l_base}
    rotation=<rot 19.873 -10.421 -1.849>
    translation=<1.933 0.14 0.004>
   }
   n10 = "Transform (Modify) Bone - Bone: dyn_coat_fr_l_01"(ComponentPose=@n9, Alpha=1, Translation="0,0,0", Rotation="0,0,0", Scale="1,1,1") {
    rotationMode=BMM_Additive
    rotationSpace=BCS_BoneSpace
    translationMode=BMM_Additive
    translationSpace=BCS_BoneSpace
    alphaBoolBlend {blendOption=Linear}
    alphaScaleBiasClamp {bInterpResult=true}
    boneToModify {boneName=dyn_coat_fr_l_01}
   }
   n11 = "Transform (Modify) Bone - Bone: dyn_coat_fr_l_02"(ComponentPose=@n10, Alpha=1, Translation="-0.390,-0.073,-0.968", Rotation="-19.999,0.052,-0.141", Scale="1,1,1") {
    rotationMode=BMM_Additive
    rotationSpace=BCS_BoneSpace
    translationMode=BMM_Additive
    translationSpace=BCS_BoneSpace
    alphaBoolBlend {blendOption=Linear}
    alphaScaleBiasClamp {bInterpResult=true}
    boneToModify {boneName=dyn_coat_fr_l_02}
    rotation=<rot -19.999 0.052 -0.141>
    translation=<-0.39 -0.073 -0.968>
   }
   n12 = "Transform (Modify) Bone - Bone: dyn_coat_side_l_01"(ComponentPose=@n11, Alpha=1, Translation="0,0,0", Rotation="0,0,0", Scale="1,1,1") {
    rotationMode=BMM_Additive
    rotationSpace=BCS_BoneSpace
    translationMode=BMM_Additive
    translationSpace=BCS_BoneSpace
    alphaBoolBlend {blendOption=Linear}
    alphaScaleBiasClamp {bInterpResult=true}
    boneToModify {boneName=dyn_coat_side_l_01}
   }
   "Transform (Modify) Bone - Bone: dyn_coat_bck_l_01"(ComponentPose=@n12, Alpha=1, Translation="0,0,0", Rotation="0,0,0", Scale="1,1,1") {
    rotationMode=BMM_Additive
    rotationSpace=BCS_BoneSpace
    translationMode=BMM_Additive
    translationSpace=BCS_BoneSpace
    alphaBoolBlend {blendOption=Linear}
    alphaScaleBiasClamp {bInterpResult=true}
    boneToModify {boneName=dyn_coat_bck_l_01}
   }
   n13 = "Transform (Modify) Bone - Bone: dyn_coat_bck_l_base"(
    ComponentPose=Local To Component(LocalPose=Use cached pose 'SimPoseCache'())
    Alpha=1
    Translation="-13.697,0.978,4.395"
    Rotation="39.636,4.423,12.232"
    Scale="1,1,1"
   ) {
    rotationMode=BMM_Additive
    rotationSpace=BCS_BoneSpace
    translationMode=BMM_Additive
    translationSpace=BCS_BoneSpace
    alphaBoolBlend {blendOption=Linear}
    alphaScaleBiasClamp {bInterpResult=true, interpSpeedDecreasing=30, interpSpeedIncreasing=30}
    boneToModify {boneName=dyn_coat_bck_l_base}
    rotation=<rot 39.636 4.423 12.232>
    translation=<-13.697 0.978 4.395>
   }
   n14 = "Transform (Modify) Bone - Bone: dyn_coat_bck_r_base"(ComponentPose=@n13, Alpha=1, Translation="-14.135,-0.162,0.227", Rotation="40.000,-0.000,0.000", Scale="1,1,1") {
    rotationMode=BMM_Additive
    rotationSpace=BCS_BoneSpace
    translationMode=BMM_Additive
    translationSpace=BCS_BoneSpace
    alphaBoolBlend {blendOption=Linear}
    alphaScaleBiasClamp {bInterpResult=true, interpSpeedDecreasing=30, interpSpeedIncreasing=30}
    boneToModify {boneName=dyn_coat_bck_r_base}
    rotation=<rot 40 0 0>
    translation=<-14.1351 -0.162 0.2269>
   }
   n15 = "Transform (Modify) Bone - Bone: dyn_coat_side_l_base"(ComponentPose=@n14, Alpha=1, Translation="0,0,0", Rotation="10.000,0.000,20.000", Scale="1,1,1") {
    rotationMode=BMM_Additive
    rotationSpace=BCS_BoneSpace
    translationMode=BMM_Additive
    translationSpace=BCS_BoneSpace
    alphaBoolBlend {blendOption=Linear}
    alphaScaleBiasClamp {bInterpResult=true, interpSpeedDecreasing=30, interpSpeedIncreasing=30}
    boneToModify {boneName=dyn_coat_side_l_base}
    rotation=<rot 10 0 20>
   }
   n16 = "Transform (Modify) Bone - Bone: dyn_coat_side_r_base"(ComponentPose=@n15, Alpha=1, Translation="0,0,0", Rotation="9.847,-1.754,-10.151", Scale="1,1,1") {
    rotationMode=BMM_Additive
    rotationSpace=BCS_BoneSpace
    translationMode=BMM_Additive
    translationSpace=BCS_BoneSpace
    alphaBoolBlend {blendOption=Linear}
    alphaScaleBiasClamp {bInterpResult=true, interpSpeedDecreasing=30, interpSpeedIncreasing=30}
    boneToModify {boneName=dyn_coat_side_r_base}
    rotation=<rot 9.847 -1.754 -10.151>
   }
   n17 = "Constraint - dyn_coat_side_r_03 (P)"(ComponentPose=@n16, Alpha=1, ConstraintWeights_0=1) {
    alphaBoolBlend {blendOption=Linear}
    alphaScaleBiasClamp {bInterpResult=true, interpSpeedDecreasing=20, interpSpeedIncreasing=20}
    boneToModify {boneName=dyn_coat_side_r_03}
    constraintSetup [{targetBone {boneName=dyn_coat_bck_r_03}}]
    constraintWeights [1]
   }
   n18 = "Constraint - dyn_coat_bck_r_03 (P)"(ComponentPose=@n17, Alpha=1, ConstraintWeights_0=1) {
    alphaBoolBlend {blendOption=Linear}
    alphaScaleBiasClamp {bInterpResult=true, interpSpeedDecreasing=20, interpSpeedIncreasing=20}
    boneToModify {boneName=dyn_coat_bck_r_03}
    constraintSetup [{targetBone {boneName=dyn_coat_side_r_03}}]
    constraintWeights [1]
   }
   n19 = "Transform (Modify) Bone - Bone: dyn_spike_fr"(ComponentPose=@n18, Alpha=1, Translation="0.087,-0.585,0.176", Rotation="17.569,13.907,13.959", Scale="1,1,1") {
    rotationMode=BMM_Additive
    rotationSpace=BCS_BoneSpace
    translationMode=BMM_Additive
    translationSpace=BCS_BoneSpace
    alphaBoolBlend {blendOption=Linear}
    alphaScaleBiasClamp {bInterpResult=true, interpSpeedDecreasing=25, interpSpeedIncreasing=25}
    boneToModify {boneName=dyn_spike_fr}
    rotation=<rot 17.569 13.907 13.959>
    translation=<0.087 -0.585 0.176>
   }
   n20 = "Transform (Modify) Bone - Bone: dyn_coat_side_l_base"(ComponentPose=@n19, Alpha=1, Translation="-1.102,0.344,2.255", Rotation="-0.000,0.004,-0.013", Scale="1,1,1") {
    rotationMode=BMM_Additive
    rotationSpace=BCS_BoneSpace
    translationMode=BMM_Additive
    translationSpace=BCS_BoneSpace
    alphaBoolBlend {blendOption=Linear}
    alphaScaleBiasClamp {bInterpResult=true}
    boneToModify {boneName=dyn_coat_side_l_base}
    rotation=<rot 0 0.004 -0.013>
    translation=<-1.102 0.344 2.2549>
   }
   n21 = "Transform (Modify) Bone - Bone: dyn_coat_side_r_02"(ComponentPose=@n20, Alpha=1, Translation="1.219,0.734,-3.193", Rotation="29.274,8.933,-8.863", Scale="1,1,1") {
    rotationMode=BMM_Additive
    rotationSpace=BCS_BoneSpace
    translationMode=BMM_Additive
    translationSpace=BCS_BoneSpace
    alphaBoolBlend {blendOption=Linear}
    alphaScaleBiasClamp {bInterpResult=true, interpSpeedDecreasing=25, interpSpeedIncreasing=25}
    boneToModify {boneName=dyn_coat_side_r_02}
    rotation=<rot 29.274 8.933 -8.863>
    translation=<1.219 0.734 -3.193>
   }
   n22 = "Transform (Modify) Bone - Bone: dyn_coat_bck_r_03"(ComponentPose=@n21, Alpha=1, Translation="-0.441,1.405,0.210", Rotation="-0.327,1.828,59.965", Scale="1,1,1") {
    rotationMode=BMM_Additive
    rotationSpace=BCS_BoneSpace
    translationMode=BMM_Additive
    translationSpace=BCS_BoneSpace
    alphaBoolBlend {blendOption=Linear}
    alphaScaleBiasClamp {bInterpResult=true, interpSpeedDecreasing=25, interpSpeedIncreasing=25}
    boneToModify {boneName=dyn_coat_bck_r_03}
    rotation=<rot -0.327 1.828 59.965>
    translation=<-0.441 1.405 0.21>
   }
   n23 = "Transform (Modify) Bone - Bone: dyn_coat_fr_l_01"(ComponentPose=@n22, Alpha=1, Translation="-4.001,-3.989,-2.829", Rotation="0.594,3.421,-19.736", Scale="1,1,1") {
    rotationMode=BMM_Additive
    rotationSpace=BCS_BoneSpace
    translationMode=BMM_Additive
    translationSpace=BCS_BoneSpace
    alphaBoolBlend {blendOption=Linear}
    alphaScaleBiasClamp {bInterpResult=true}
    boneToModify {boneName=dyn_coat_fr_l_01}
    rotation=<rot 0.594 3.4211 -19.736>
    translation=<-4.001 -3.989 -2.829>
   }
   n24 = Anim Dynamics - Bone: dyn_coat_fr_l_02!(ComponentPose=@n23, Alpha=1, GravityOverride:Vector="0,0,-980", ExternalForce="0,0,0") -> (Pose) {
    angularDampingOverride=1
    angularSpringConstant=20000
    bAngularSpring=true
    bOverrideAngularDamping=true
    bUseGravityOverride=true
    gravityScale=0.35
    simulationSpace=BoneRelative
    alphaBoolBlend {blendOption=Linear}
    boundBone {boneName=dyn_coat_fr_l_02}
    componentLinearAccScale=<0.8 0.8 0.8>
    componentLinearVelScale=<0.8 0.8 0.8>
    gravityOverride {z=-980}
    physicsBodyDefinitions [{
      boundBone {boneName=dyn_coat_fr_l_02}
      boxExtents=<20 30 30>
      constraintSetup {angularLimitsMax=<0 2 0>, angularLimitsMin=<0 -20 0>, angularTarget=<1 0 0>}
      localJointOffset=<4 0 0>
     }]
    relativeSpaceBone {boneName=root}
   }
   n25 = Anim Dynamics - Bone: dyn_coat_fr_l_03!(ComponentPose=@n24, Alpha=1, GravityOverride:Vector="0,0,-980", ExternalForce="0,0,0") -> (Pose) {
    angularDampingOverride=1
    angularSpringConstant=20000
    bAngularSpring=true
    bOverrideAngularDamping=true
    bUseGravityOverride=true
    gravityScale=0.35
    simulationSpace=BoneRelative
    alphaBoolBlend {blendOption=Linear}
    boundBone {boneName=dyn_coat_fr_l_03}
    componentLinearAccScale=<0.8 0.8 0.8>
    componentLinearVelScale=<0.8 0.8 0.8>
    gravityOverride {z=-980}
    physicsBodyDefinitions [{
      boundBone {boneName=dyn_coat_fr_l_03}
      boxExtents=<20 30 30>
      constraintSetup {angularLimitsMax=<0 5 0>, angularLimitsMin=<0 -20 0>, angularTarget=<1 0 0>}
      localJointOffset=<4 0 0>
     }]
    relativeSpaceBone {boneName=root}
   }
   n26 = Anim Dynamics - Bone: dyn_coat_side_l_02!(ComponentPose=@n25, Alpha=1, GravityOverride:Vector="0,0,-980", ExternalForce="0,0,0") -> (Pose) {
    angularDampingOverride=1
    angularSpringConstant=20000
    bAngularSpring=true
    bOverrideAngularDamping=true
    bUseGravityOverride=true
    gravityScale=0.35
    simulationSpace=BoneRelative
    alphaBoolBlend {blendOption=Linear}
    boundBone {boneName=dyn_coat_side_l_02}
    componentLinearAccScale=<0.8 0.8 0.8>
    componentLinearVelScale=<0.8 0.8 0.8>
    gravityOverride {z=-980}
    physicsBodyDefinitions [{
      boundBone {boneName=dyn_coat_side_l_02}
      boxExtents=<20 30 30>
      constraintSetup {angularLimitsMax=<0 2 0>, angularLimitsMin=<0 -20 0>, angularTarget=<1 0 0>}
      localJointOffset=<4 0 0>
     }]
    relativeSpaceBone {boneName=root}
   }
   n27 = Anim Dynamics - Bone: dyn_coat_side_l_03!(ComponentPose=@n26, Alpha=1, GravityOverride:Vector="0,0,-980", ExternalForce="0,0,0") -> (Pose) {
    angularDampingOverride=1
    angularSpringConstant=20000
    bAngularSpring=true
    bOverrideAngularDamping=true
    bUseGravityOverride=true
    gravityScale=0.35
    simulationSpace=BoneRelative
    alphaBoolBlend {blendOption=Linear}
    boundBone {boneName=dyn_coat_side_l_03}
    componentLinearAccScale=<0.8 0.8 0.8>
    componentLinearVelScale=<0.8 0.8 0.8>
    gravityOverride {z=-980}
    physicsBodyDefinitions [{
      boundBone {boneName=dyn_coat_side_l_03}
      boxExtents=<20 30 30>
      constraintSetup {angularLimitsMax=<0 2 0>, angularLimitsMin=<0 -20 0>, angularTarget=<1 0 0>}
      localJointOffset=<4 0 0>
     }]
    relativeSpaceBone {boneName=root}
   }
   n28 = Anim Dynamics - Bone: dyn_coat_bck_l_02!(ComponentPose=@n27, Alpha=1, GravityOverride:Vector="0,0,-980", ExternalForce="0,0,0") -> (Pose) {
    angularDampingOverride=1
    angularSpringConstant=20000
    bAngularSpring=true
    bOverrideAngularDamping=true
    bUseGravityOverride=true
    gravityScale=0.35
    simulationSpace=BoneRelative
    alphaBoolBlend {blendOption=Linear}
    boundBone {boneName=dyn_coat_bck_l_02}
    componentLinearAccScale=<0.8 0.8 0.8>
    componentLinearVelScale=<0.8 0.8 0.8>
    gravityOverride {z=-980}
    physicsBodyDefinitions [{
      boundBone {boneName=dyn_coat_bck_l_02}
      boxExtents=<20 30 30>
      constraintSetup {angularLimitsMax=<0 2 0>, angularLimitsMin=<0 -20 0>, angularTarget=<1 0 0>}
      localJointOffset=<4 0 0>
     }]
    relativeSpaceBone {boneName=root}
   }
   n29 = Anim Dynamics - Bone: dyn_coat_bck_l_03!(ComponentPose=@n28, Alpha=1, GravityOverride:Vector="0,0,-980", ExternalForce="0,0,0") -> (Pose) {
    angularDampingOverride=1
    angularSpringConstant=20000
    bAngularSpring=true
    bOverrideAngularDamping=true
    bUseGravityOverride=true
    gravityScale=0.35
    simulationSpace=BoneRelative
    alphaBoolBlend {blendOption=Linear}
    boundBone {boneName=dyn_coat_bck_l_03}
    componentLinearAccScale=<0.8 0.8 0.8>
    componentLinearVelScale=<0.8 0.8 0.8>
    gravityOverride {z=-980}
    physicsBodyDefinitions [{
      boundBone {boneName=dyn_coat_bck_l_03}
      boxExtents=<20 30 30>
      constraintSetup {angularLimitsMax=<0 2 0>, angularLimitsMin=<0 -20 0>, angularTarget=<1 0 0>}
      localJointOffset=<4 0 0>
     }]
    relativeSpaceBone {boneName=root}
   }
   n30 = Anim Dynamics - Bone: dyn_coat_fr_r_02!(ComponentPose=@n29, Alpha=1, GravityOverride:Vector="0,0,-980", ExternalForce="0,0,0") -> (Pose) {
    angularDampingOverride=1
    angularSpringConstant=20000
    bAngularSpring=true
    bOverrideAngularDamping=true
    bUseGravityOverride=true
    gravityScale=0.35
    simulationSpace=BoneRelative
    alphaBoolBlend {blendOption=Linear}
    boundBone {boneName=dyn_coat_fr_r_02}
    componentLinearAccScale=<0.8 0.8 0.8>
    componentLinearVelScale=<0.8 0.8 0.8>
    gravityOverride {z=-980}
    physicsBodyDefinitions [{
      boundBone {boneName=dyn_coat_fr_r_02}
      boxExtents=<20 30 30>
      constraintSetup {angularLimitsMax=<0 5 0>, angularLimitsMin=<0 -20 0>, angularTarget=<1 0 0>}
      localJointOffset=<4 0 0>
     }]
    relativeSpaceBone {boneName=root}
   }
   n31 = Anim Dynamics - Bone: dyn_coat_fr_r_03!(ComponentPose=@n30, Alpha=1, GravityOverride:Vector="0,0,-980", ExternalForce="0,0,0") -> (Pose) {
    angularDampingOverride=1
    angularSpringConstant=20000
    bAngularSpring=true
    bOverrideAngularDamping=true
    bUseGravityOverride=true
    gravityScale=0.35
    simulationSpace=BoneRelative
    alphaBoolBlend {blendOption=Linear}
    boundBone {boneName=dyn_coat_fr_r_03}
    componentLinearAccScale=<0.8 0.8 0.8>
    componentLinearVelScale=<0.8 0.8 0.8>
    gravityOverride {z=-980}
    physicsBodyDefinitions [{
      boundBone {boneName=dyn_coat_fr_r_03}
      boxExtents=<20 30 30>
      constraintSetup {angularLimitsMax=<0 5 0>, angularLimitsMin=<0 -20 0>, angularTarget=<1 0 0>}
      localJointOffset=<4 0 0>
     }]
    relativeSpaceBone {boneName=root}
   }
   n32 = Anim Dynamics - Bone: dyn_coat_side_r_02!(ComponentPose=@n31, Alpha=1, GravityOverride:Vector="0,0,-980", ExternalForce="0,0,0") -> (Pose) {
    angularDampingOverride=1
    angularSpringConstant=20000
    bAngularSpring=true
    bOverrideAngularDamping=true
    bUseGravityOverride=true
    gravityScale=0.35
    simulationSpace=BoneRelative
    alphaBoolBlend {blendOption=Linear}
    boundBone {boneName=dyn_coat_side_r_02}
    componentLinearAccScale=<0.8 0.8 0.8>
    componentLinearVelScale=<0.8 0.8 0.8>
    gravityOverride {z=-980}
    physicsBodyDefinitions [{
      boundBone {boneName=dyn_coat_side_r_02}
      boxExtents=<20 30 30>
      constraintSetup {angularLimitsMax=<0 0 0>, angularLimitsMin=<0 -20 0>, angularTarget=<1 0 0>}
      localJointOffset=<4 0 0>
     }]
    relativeSpaceBone {boneName=root}
   }
   n33 = Anim Dynamics - Bone: dyn_coat_side_r_03!(ComponentPose=@n32, Alpha=1, GravityOverride:Vector="0,0,-980", ExternalForce="0,0,0") -> (Pose) {
    angularDampingOverride=1
    angularSpringConstant=20000
    bAngularSpring=true
    bOverrideAngularDamping=true
    bUseGravityOverride=true
    gravityScale=0.35
    simulationSpace=BoneRelative
    alphaBoolBlend {blendOption=Linear}
    boundBone {boneName=dyn_coat_side_r_03}
    componentLinearAccScale=<0.8 0.8 0.8>
    componentLinearVelScale=<0.8 0.8 0.8>
    gravityOverride {z=-980}
    physicsBodyDefinitions [{
      boundBone {boneName=dyn_coat_side_r_03}
      boxExtents=<20 30 30>
      constraintSetup {angularLimitsMax=<0 0 0>, angularLimitsMin=<0 -20 0>, angularTarget=<1 0 0>}
      localJointOffset=<4 0 0>
     }]
    relativeSpaceBone {boneName=root}
   }
   n34 = Anim Dynamics - Bone: dyn_coat_bck_r_02!(ComponentPose=@n33, Alpha=1, GravityOverride:Vector="0,0,-980", ExternalForce="0,0,0") -> (Pose) {
    angularDampingOverride=1
    angularSpringConstant=20000
    bAngularSpring=true
    bOverrideAngularDamping=true
    bUseGravityOverride=true
    gravityScale=0.35
    simulationSpace=BoneRelative
    alphaBoolBlend {blendOption=Linear}
    boundBone {boneName=dyn_coat_bck_r_02}
    componentLinearAccScale=<0.8 0.8 0.8>
    componentLinearVelScale=<0.8 0.8 0.8>
    gravityOverride {z=-980}
    physicsBodyDefinitions [{
      boundBone {boneName=dyn_coat_bck_r_02}
      boxExtents=<20 30 30>
      constraintSetup {angularLimitsMax=<0 2 0>, angularLimitsMin=<0 -20 0>, angularTarget=<1 0 0>}
      localJointOffset=<4 0 0>
     }]
    relativeSpaceBone {boneName=root}
   }
   n35 = Anim Dynamics - Bone: dyn_coat_bck_r_03!(ComponentPose=@n34, Alpha=1, GravityOverride:Vector="0,0,-980", ExternalForce="0,0,0") -> (Pose) {
    angularDampingOverride=1
    angularSpringConstant=20000
    bAngularSpring=true
    bOverrideAngularDamping=true
    bUseGravityOverride=true
    gravityScale=0.35
    simulationSpace=BoneRelative
    alphaBoolBlend {blendOption=Linear}
    boundBone {boneName=dyn_coat_bck_r_03}
    componentLinearAccScale=<0.8 0.8 0.8>
    componentLinearVelScale=<0.8 0.8 0.8>
    gravityOverride {z=-980}
    physicsBodyDefinitions [{
      boundBone {boneName=dyn_coat_bck_r_03}
      boxExtents=<20 30 30>
      constraintSetup {angularLimitsMax=<0 0 0>, angularLimitsMin=<0 -20 0>, angularTarget=<1 0 0>}
      localJointOffset=<4 0 0>
     }]
    relativeSpaceBone {boneName=root}
   }
   Output Pose(Result=Component To Local(ComponentPose=@n35)) {layerGroup=None}
   n36 = Control Rig(Source=Copy Pose From Mesh() {bUseAttachedParent=true}) {
    alphaBoolBlend {blendOption=Linear}
    controlRigAssetReference {blueprintRigClass=&/M_MED_ProdigyHaughty_DeformRig_ControlRig.M_MED_ProdigyHaughty_DeformRig_ControlRig_C}
   }
   n37 = RigidBody!(ComponentPose=Local To Component(LocalPose=@n36), Alpha, OverrideWorldGravity:Vector="0,0,0", ExternalForce) -> (Pose) {
    bClampLinearTranslationLimitToRefPose=true
    bOverrideWorldGravity=true
    overridePhysicsAsset=&/M_MED_ProdigyHaughty_Physics
    simulationSpace=BaseBoneSpace
    alphaBoolBlend {blendOption=Linear}
    baseBoneRef {boneName=pelvis}
    componentAppliedLinearAccClamp=<5000 5000 5000>
    componentLinearAccScale=<0.1 0.1 0.1>
    simSpaceSettings {maxAngularAcceleration=1000, maxAngularVelocity=1000, maxLinearAcceleration=1000, maxLinearVelocity=1000}
   }
   n38 = "Constraint - dyn_spike_fr (P)"(ComponentPose=@n37, Alpha=1, ConstraintWeights_0=1) {
    alphaBoolBlend {blendOption=Linear}
    alphaScaleBiasClamp {bInterpResult=true, interpSpeedDecreasing=20, interpSpeedIncreasing=20}
    boneToModify {boneName=dyn_spike_fr}
    constraintSetup [{targetBone {boneName=thigh_l}}]
    constraintWeights [1]
   }
   n39 = "Constraint - dyn_lapel_l (PP)"!(ComponentPose=@n38, Alpha=1, ConstraintWeights_0=0.75, ConstraintWeights_1:float=0.25) -> (Pose) {
    alphaBoolBlend {blendOption=Linear}
    boneToModify {boneName=dyn_lapel_l}
    constraintSetup [{targetBone {boneName=spine_05}}, {targetBone {boneName=clavicle_l}}]
    constraintWeights [1,1]
   }
   n40 = "Constraint - dyn_lapel_r (PP)"!(ComponentPose=@n39, Alpha=1, ConstraintWeights_0=0.5, ConstraintWeights_1:float=0.5) -> (Pose) {
    alphaBoolBlend {blendOption=Linear}
    boneToModify {boneName=dyn_lapel_r}
    constraintSetup [{targetBone {boneName=spine_05}}, {targetBone {boneName=clavicle_r}}]
    constraintWeights [1,1]
   }
   n41 = Anim Dynamics - Bone: dyn_necklace_01!(ComponentPose=@n40, Alpha=1, GravityOverride:Vector="0,0,-350", ExternalForce="0,0,0") -> (Pose) {
    angularDampingOverride=0.8
    angularSpringConstant=100
    bGravityOverrideInSimSpace=true
    bUseGravityOverride=true
    linearDampingOverride=0.9
    linearSpringConstant=150
    numSolverIterationsPreUpdate=3
    alphaBoolBlend {blendOption=Linear}
    boundBone {boneName=dyn_necklace_01}
    componentAppliedLinearAccClamp=<10000 10000 10000>
    componentLinearAccScale=<0.1 0.1 0.1>
    componentLinearVelScale=<0.1 0.1 0.1>
    gravityOverride {z=-350}
    physicsBodyDefinitions [{
      boundBone {boneName=dyn_necklace_01}
      boxExtents=<15 15 15>
      constraintSetup {angularLimitsMax=<0 -2 5>, angularLimitsMin=<0 -40 -5>, angularTarget=<1 0 0>}
      localJointOffset=<2 0 0>
     }]
    relativeSpaceBone {boneName=spine_05}
   }
   n42 = Anim Dynamics - Bone: dyn_necklace_02!(ComponentPose=@n41, Alpha=1, GravityOverride:Vector="0,0,-200", ExternalForce="0,0,0") -> (Pose) {
    angularDampingOverride=0.8
    angularSpringConstant=100
    bGravityOverrideInSimSpace=true
    bUseGravityOverride=true
    linearDampingOverride=0.9
    linearSpringConstant=150
    numSolverIterationsPreUpdate=3
    alphaBoolBlend {blendOption=Linear}
    boundBone {boneName=dyn_necklace_02}
    componentAppliedLinearAccClamp=<10000 10000 10000>
    componentLinearAccScale=<0.1 0.1 0.1>
    componentLinearVelScale=<0.1 0.1 0.1>
    gravityOverride {z=-200}
    physicsBodyDefinitions [{
      boundBone {boneName=dyn_necklace_02}
      boxExtents=<15 15 15>
      constraintSetup {angularLimitsMax=<0 -2 5>, angularLimitsMin=<0 -40 -5>, angularTarget=<1 0 0>}
      localJointOffset=<1 0 0>
     }]
    relativeSpaceBone {boneName=spine_05}
   }
   n43 = Anim Dynamics - Bone: dyn_necklace_03!(ComponentPose=@n42, Alpha=1, GravityOverride:Vector="0,0,-200", ExternalForce="0,0,0") -> (Pose) {
    angularDampingOverride=0.8
    angularSpringConstant=100
    bGravityOverrideInSimSpace=true
    bUseGravityOverride=true
    linearDampingOverride=0.9
    linearSpringConstant=150
    numSolverIterationsPreUpdate=3
    alphaBoolBlend {blendOption=Linear}
    boundBone {boneName=dyn_necklace_03}
    componentAppliedLinearAccClamp=<10000 10000 10000>
    componentLinearAccScale=<0.1 0.1 0.1>
    componentLinearVelScale=<0.1 0.1 0.1>
    gravityOverride {z=-200}
    physicsBodyDefinitions [{
      boundBone {boneName=dyn_necklace_03}
      boxExtents=<15 15 15>
      constraintSetup {angularLimitsMax=<0 -2 5>, angularLimitsMin=<0 -40 -5>, angularTarget=<1 0 0>}
      localJointOffset=<1 0 0>
     }]
    relativeSpaceBone {boneName=spine_05}
   }
   n44 = "Constraint - dyn_necklace_01 (P)"(ComponentPose=@n43, Alpha=0.5, ConstraintWeights_0=1) {
    alphaBoolBlend {blendOption=Linear}
    boneToModify {boneName=dyn_necklace_01}
    constraintSetup [{targetBone {boneName=spine_04}}]
    constraintWeights [1]
   }
   n45 = Anim Dynamics - Bone: dyn_jaggedBracelets!(ComponentPose=@n44, Alpha=1, GravityOverride:Vector="0,0,-350", ExternalForce="0,0,0") -> (Pose) {
    angularDampingOverride=0.8
    angularSpringConstant=100
    bAngularSpring=true
    bGravityOverrideInSimSpace=true
    bUseGravityOverride=true
    linearDampingOverride=0.9
    linearSpringConstant=150
    numSolverIterationsPreUpdate=3
    simulationSpace=BoneRelative
    alphaBoolBlend {blendOption=Linear}
    boundBone {boneName=dyn_jaggedBracelets}
    componentAppliedLinearAccClamp=<10000 10000 10000>
    gravityOverride {z=-350}
    physicsBodyDefinitions [{
      boundBone {boneName=dyn_jaggedBracelets}
      boxExtents=<15 15 15>
      constraintSetup {angularLimitsMax=<1 1 1>, angularLimitsMin=<-1 -1 -1>, angularTarget=<1 0 0>, linearAxesMin=<-2 0 0>}
      localJointOffset=<2 0 0>
     }]
    relativeSpaceBone {boneName=lowerarm_l}
   }
   n46 = Anim Dynamics - Bone: dyn_LoopBracelet_01!(ComponentPose=@n45, Alpha=1, GravityOverride:Vector="0,0,-200", ExternalForce="0,0,0") -> (Pose) {
    angularDampingOverride=0.8
    angularSpringConstant=100
    bAngularSpring=true
    bGravityOverrideInSimSpace=true
    bUseGravityOverride=true
    linearDampingOverride=0.9
    linearSpringConstant=150
    numSolverIterationsPreUpdate=3
    simulationSpace=BoneRelative
    alphaBoolBlend {blendOption=Linear}
    boundBone {boneName=dyn_LoopBracelet_01}
    componentAppliedLinearAccClamp=<10000 10000 10000>
    gravityOverride {z=-200}
    physicsBodyDefinitions [{
      boundBone {boneName=dyn_LoopBracelet_01}
      boxExtents=<15 15 15>
      constraintSetup {angularLimitsMax=<1 1 1>, angularLimitsMin=<-1 -1 -1>, angularTarget=<1 0 0>, linearAxesMin=<-2 0 0>}
      localJointOffset=<1 0 0>
     }]
    relativeSpaceBone {boneName=lowerarm_l}
   }
   n47 = Anim Dynamics - Bone: dyn_LoopBracelet_02!(ComponentPose=@n46, Alpha=1, GravityOverride:Vector="0,0,-200", ExternalForce="0,0,0") -> (Pose) {
    angularDampingOverride=0.8
    angularSpringConstant=100
    bAngularSpring=true
    bGravityOverrideInSimSpace=true
    bUseGravityOverride=true
    linearDampingOverride=0.9
    linearSpringConstant=150
    numSolverIterationsPreUpdate=3
    simulationSpace=BoneRelative
    alphaBoolBlend {blendOption=Linear}
    boundBone {boneName=dyn_LoopBracelet_02}
    componentAppliedLinearAccClamp=<10000 10000 10000>
    gravityOverride {z=-200}
    physicsBodyDefinitions [{
      boundBone {boneName=dyn_LoopBracelet_02}
      boxExtents=<15 15 15>
      constraintSetup {angularLimitsMax=<1 1 1>, angularLimitsMin=<-1 -1 -1>, angularTarget=<1 0 0>, linearAxesMin=<-2 0 0>}
      localJointOffset=<1 0 0>
     }]
    relativeSpaceBone {boneName=lowerarm_l}
   }
   Save cached pose 'SimPoseCache'(Pose=Component To Local(ComponentPose=@n47)) {cacheName=SimPoseCache}
  ]}
 EventGraph {nodes [
   Event BlueprintUpdateAnimation()
   > n1 = Sequence!(execute) -> (then_0, then_1, then_2:Exec, then_3:Exec)
   > n2 = Gravity Override State Function DEPRECATED(CharacterPartAnimInstance=Self-Reference(), GameTimeInSeconds=GetGameTimeInSeconds(), GravityOverrideParams=Get Gravity Override Params())
   > n3 = Branch(Condition=Get bIsPlayingEmote())
   > Set Gravity Override(Gravity Override=vector * vector(A=@n2, B:Vector="0,0,2") -> (ReturnValue=~))
   Set Gravity Override(execute=@n3.else, Gravity Override=@n2)
   n4 = Get bIsAccelerating2D()
   n5 = "To Float (Boolean)"(InBool=AND Boolean(A=Get bIsCrouching(), B=NOT Boolean(A=@n4)))
   Set isCrouchNotMoving(execute=@n1.then_1, isCrouchNotMoving:double=@n5) -> (Output_Get:double)
   Reroute Node()
   n6 = GetOwnerSkeletalMeshComponent()
   n7 = BreakRotator(InRot=BreakTransform(InTransform=GetDeltaTransformFromRefPose(self=@n6, BoneName=thigh_l, BaseName=pelvis)).Rotation)
   n8 = MapRangeClamped(Value=@n7.Yaw, InRangeA=0, InRangeB=-60)
   > n9 = Set L leg up(L leg up:double=@n8) -> (Output_Get:double)
   Set L leg pitch(
    execute=[@n9.then, @n13.then, @n14.then]
    L leg pitch:double=SelectFloat(A=0, B=MapRangeClamped(Value=@n7.Pitch, InRangeA=0, InRangeB=40), bPickA=Get bIsCrouching())
   ) -> (Output_Get:double)
   n10 = BreakRotator(InRot=BreakTransform(InTransform=GetDeltaTransformFromRefPose(self=@n6, BoneName=thigh_r, BaseName=pelvis)).Rotation)
   n11 = MapRangeClamped(Value=@n10.Yaw, InRangeA=0, InRangeB=-100)
   n12 = Get bIsCrouching()
   > Set R leg up(R leg up:double=SelectFloat(A=0, B=@n11, bPickA=@n12)) -> (Output_Get:double)
   > Set R Leg Up Crouch(R Leg Up Crouch:double=SelectFloat(A=@n11, B=0, bPickA=@n12)) -> (Output_Get:double)
   > Set R leg pitch(R leg pitch:double=MapRangeClamped(Value=@n10.Pitch, InRangeA=0, InRangeB=35)) -> (Output_Get:double)
   > Set R leg negative pitch(R leg negative pitch:double=SelectFloat(A=MapRangeClamped(Value=@n10.Pitch, InRangeA=3, InRangeB=-15), B=0, bPickA=Get bIsCrouching())) -> (Output_Get:double)
   n13 = Set L leg up crouch(L leg up crouch:double=0) -> (Output_Get:double)
   Set L leg up crouch(L leg up crouch=@n8)
   > n14 = Set L leg up(L leg up=0)
   OR Boolean(A=Get bIsJumping(), B=Get bIsFalling())
   Branch()
   float * float(A:Wildcard, B:Wildcard) -> (ReturnValue:Wildcard) {is=title K2Node_PromotableOperator}
   float * float(A:double=@n5, B:double=0.8) -> (ReturnValue:double)
  ]}
]
variables [
 Arm Up R {type=number}
 crouch alpha {type=number}
 crouch targeting up {type=number}
 Gravity Override {type=Vector}
 Gravity Override Params {type=GravityOverrideParamsStruct}
 Head Down Targeting {type=number}
 Head Up {type=number}
 Head Up Targeting {type=number}
 isCrouchNotMoving {type=number}
 L leg pitch {type=number}
 L leg up {type=number}
 L leg up crouch {type=number}
 R leg negative pitch {type=number}
 R leg pitch {type=number}
 R leg up {type=number}
 R Leg Up Crouch {type=number}
 rban alpha {type=number}
]
